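# Sound cue for the UI toggle sound during airstrike placement, judging by its name.
# A single wave player plays the front-end UI wave fort_ui_new_chris_16 directly, with no randomization or modulation.
# It routes to the non-critical in-game UI sound class and overrides concurrency so that new instances are blocked once the limit is reached.
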SoundCue /Game/Sounds/Fort_UI_Sounds/Gameplay/Airstrike_Placement_UI_Toggle_Cue
root=WavePlayer_2
nodes [{id=WavePlayer_2, properties {soundWaveAssetPtr=/Game/Sounds/Fort_UI_Sounds/FrontEnd/fort_ui_new_chris_16}}]
properties {
 bOverrideConcurrency=true
 maxDistance=2097152
 soundClassObject=&Fort_UI_InGame_NonCritical
 concurrencyOverrides {resolutionRule=PreventNew}
}
