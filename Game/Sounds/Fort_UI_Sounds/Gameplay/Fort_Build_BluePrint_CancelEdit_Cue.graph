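# Sound cue for the Fortnite UI sound that, by its name, plays when a blueprint build edit is cancelled.
# A single front-end UI wave player runs through a modulator that randomly raises its pitch by roughly half an octave.
# It plays non-spatialized in the critical in-game UI sound class at reduced volume, with a concurrency override that stops the oldest instance beyond a small limit.
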SoundCue /Game/Sounds/Fort_UI_Sounds/Gameplay/Fort_Build_BluePrint_CancelEdit_Cue
root=Modulator_3
nodes [
 {id=Modulator_3, children [WavePlayer_24], properties {pitchMax=1.5, pitchMin=1.4}}
 {id=WavePlayer_24, properties {soundWaveAssetPtr=/Game/Sounds/Fort_UI_Sounds/FrontEnd/fort_ui_new_chris_04}}
]
properties {
 bOverrideConcurrency=true
 maxDistance=2097152
 soundClassObject=&Fort_UI_InGame_Critical
 volumeMultiplier=0.45
 attenuationOverrides {bSpatialize=false, falloffDistance=2000, nonSpatializedRadiusStart=200}
 concurrencyOverrides {maxCount=2, resolutionRule=StopOldest}
}
