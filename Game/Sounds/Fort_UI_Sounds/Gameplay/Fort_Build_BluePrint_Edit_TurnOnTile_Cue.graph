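# Sound cue for the UI sound of turning on a tile during building blueprint editing.
# A single wave from the front-end UI sounds plays through a Modulator that randomizes pitch slightly upward for variation.
# It plays on the in-game critical UI sound class at reduced volume, with spatialization off in its attenuation overrides and a concurrency override that caps simultaneous plays by stopping the oldest.
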SoundCue /Game/Sounds/Fort_UI_Sounds/Gameplay/Fort_Build_BluePrint_Edit_TurnOnTile_Cue
root=Modulator_3
nodes [
 {id=Modulator_3, children [WavePlayer_0], properties {pitchMax=1.1, pitchMin=1}}
 {id=WavePlayer_0, properties {soundWaveAssetPtr=/Game/Sounds/Fort_UI_Sounds/FrontEnd/fort_ui_new_chris_21a}}
]
properties {
 bOverrideConcurrency=true
 maxDistance=2097152
 soundClassObject=&Fort_UI_InGame_Critical
 volumeMultiplier=0.42
 attenuationOverrides {bSpatialize=false, falloffDistance=2000, nonSpatializedRadiusStart=1000}
 concurrencyOverrides {maxCount=2, resolutionRule=StopOldest}
}
