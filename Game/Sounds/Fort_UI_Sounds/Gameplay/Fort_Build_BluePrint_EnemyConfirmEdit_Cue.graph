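# Sound cue for the confirmation sound of an enemy's building edit, judging by its name and wave assets.
# A random node picks one of three equally weighted waves and feeds a modulator at the root.
# It plays at reduced volume with overridden short-range attenuation and concurrency overrides set to two voices, stopping the oldest.
# It is assigned to the melee weapon sound class rather than a UI one.
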SoundCue /Game/Sounds/Fort_UI_Sounds/Gameplay/Fort_Build_BluePrint_EnemyConfirmEdit_Cue
root=Modulator_0
nodes [
 {id=Modulator_0, children [Random_0]}
 {id=Random_0, children [WavePlayer_0,WavePlayer_1,WavePlayer_2], properties {weights [1,1,1]}}
 {id=WavePlayer_0, properties {soundWaveAssetPtr=/Game/Sounds/Fort_UI_Sounds/FrontEnd/Build_Edit_EnemyConfirm_01}}
 {id=WavePlayer_1, properties {soundWaveAssetPtr=/Game/Sounds/Fort_UI_Sounds/FrontEnd/Build_Edit_EnemyConfirm_03}}
 {id=WavePlayer_2, properties {soundWaveAssetPtr=/Game/Sounds/Fort_UI_Sounds/FrontEnd/Build_Edit_EnemyConfirm_02}}
]
properties {
 bOverrideAttenuation=true
 maxDistance=1024
 soundClassObject=&Fort_WeaponMelee_PSM
 volumeMultiplier=0.6
 attenuationOverrides {falloffDistance=512, stereoSpread=256, attenuationShapeExtents {x=512}}
 concurrencyOverrides {maxCount=2, resolutionRule=StopOldest}
}
