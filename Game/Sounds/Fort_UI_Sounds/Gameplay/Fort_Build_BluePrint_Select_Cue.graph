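# Sound cue for a UI gameplay sound, named for selecting a blueprint when building.
# A single short wave player feeds a Modulator left at default settings, which slightly varies pitch and volume on each play.
# It plays at reduced volume in the critical in-game UI sound class.
# Its own concurrency override allows one instance at a time, stopping the oldest so a new selection cuts off the last.
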SoundCue /Game/Sounds/Fort_UI_Sounds/Gameplay/Fort_Build_BluePrint_Select_Cue
root=Modulator_3
nodes [
 {id=Modulator_3, children [WavePlayer_11]}
 {id=WavePlayer_11, properties {soundWaveAssetPtr=&/Fort_Build_BluePrint_Select_02_short}}
]
properties {
 bOverrideConcurrency=true
 maxDistance=2097152
 soundClassObject=&Fort_UI_InGame_Critical
 volumeMultiplier=0.45
 attenuationOverrides {bSpatialize=false, falloffDistance=2000, nonSpatializedRadiusStart=1000}
 concurrencyOverrides {maxCount=1, resolutionRule=StopOldest}
}
