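# Sound cue for build completion, routed to the non-critical player construction sound class.
# A root mixer layers three modulated waves at different levels: a construction-done sound, a briefly delayed bell and a front-end big-menu UI sound.
# The bell and menu layers have a raised upper pitch limit for random variation.
# Overridden attenuation spatializes the cue, and a shared concurrency rule limits simultaneous instances.
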SoundCue /Game/Sounds/Fort_UI_Sounds/Gameplay/Fort_Build_Done_Cue
root=Mixer_3
nodes [
 {id=Delay_5, children [Modulator_12], properties {delayMax=0.2, delayMin=0.2}}
 {id=Mixer_3, children [Modulator_9,Delay_5,Modulator_13], properties {inputVolume [0.6,1,1.3]}}
 {id=Modulator_12, children [WavePlayer_51], properties {pitchMax=1.5}}
 {id=Modulator_13, children [WavePlayer_52], properties {pitchMax=1.5}}
 {id=Modulator_9, children [WavePlayer_50]}
 {id=WavePlayer_50, properties {soundWaveAssetPtr=&/Fort_Contructino_Done_02}}
 {id=WavePlayer_51, properties {soundWaveAssetPtr=&/bell_tr_01}}
 {id=WavePlayer_52, properties {soundWaveAssetPtr=/Game/Sounds/Fort_UI_Sounds/FrontEnd/fort_ui_BigMenu_03}}
]
properties {
 bOverrideAttenuation=true
 maxDistance=1500
 soundClassObject=/Game/Sounds/FortSoundClassesAndModes/Fort_Player_Construction_NonCritical
 volumeMultiplier=0.3
 attenuationOverrides {falloffDistance=1000, stereoSpread=450, attenuationShapeExtents {x=500}}
 concurrencyOverrides {resolutionRule=PreventNew}
 concurrencySet [/Game/Sounds/Concurrency_Rules/ThreeMaximum_StopOldest_VolumeScale]
}
