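# Sound cue for a build piece fit sound, apparently variant 13, in the gameplay UI sound set.
# A single wave player feeds the Fort_Build_PieceFit_13 wave straight to the output, with no randomization or modulation nodes.
# It routes to the player construction sound class and overrides attenuation for its falloff distance.
# Its own concurrency override prevents new instances from starting once the limit is reached.
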SoundCue /Game/Sounds/Fort_UI_Sounds/Gameplay/Fort_Build_PieceFit_13.Fort_Build_PieceFit_13_Cue
root=WavePlayer_1963
nodes [{id=WavePlayer_1963, properties {soundWaveAssetPtr=&/Fort_Build_PieceFit_13}}]
properties {
 bOverrideAttenuation=true
 bOverrideConcurrency=true
 maxDistance=3000
 soundClassObject=/Game/Sounds/FortSoundClassesAndModes/Fort_Player_Construction
 attenuationOverrides {falloffDistance=2600}
 concurrencyOverrides {resolutionRule=PreventNew}
}
