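# Sound cue for the build repair-complete effect, routed through the player construction sound class.
# A mixer layers a trap build sound, pitched up to double and mixed at half volume, over a full-volume door build sound.
# An outer modulator wraps the mix with its pitch pinned at normal.
# The cue overrides attenuation for distance falloff and overrides concurrency to block new instances at the limit.
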SoundCue /Game/Sounds/Fort_UI_Sounds/Gameplay/Fort_Build_RepairComplete_Cue
root=Modulator_0
nodes [
 {id=Mixer_0, children [Modulator_1,WavePlayer_5], properties {inputVolume [0.5,1]}}
 {id=Modulator_0, children [Mixer_0], properties {pitchMax=1, pitchMin=1}}
 {id=Modulator_1, children [WavePlayer_4], properties {pitchMax=2, pitchMin=2}}
 {id=WavePlayer_4, properties {soundWaveAssetPtr=&/Fort_Build_Trap_01}}
 {id=WavePlayer_5, properties {soundWaveAssetPtr=&/Fort_Build_Door_01}}
]
properties {
 bOverrideAttenuation=true
 bOverrideConcurrency=true
 maxDistance=3000
 soundClassObject=/Game/Sounds/FortSoundClassesAndModes/Fort_Player_Construction
 attenuationOverrides {falloffDistance=2600}
 concurrencyOverrides {resolutionRule=PreventNew}
}
