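# Sound cue for the start of building repair, routed to the player construction sound class.
# A quality-level switch sends its first two levels to the standard mix wave and its last two to the "opt" variant.
# Each wave passes through its own modulator for slight variation, with the pitch floor held at normal.
# Attenuation and concurrency are overridden on the cue, with new plays blocked once the concurrency limit is reached.
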SoundCue /Game/Sounds/Fort_UI_Sounds/Gameplay/Fort_Build_RepairStart_Cue
root=QualityLevel_0
nodes [
 {id=Modulator_0, children [WavePlayer_2], properties {pitchMin=1}}
 {id=Modulator_1, children [WavePlayer_4], properties {pitchMin=1}}
 {id=QualityLevel_0, children [Modulator_0,Modulator_0,Modulator_1,Modulator_1]}
 {id=WavePlayer_2, properties {soundWaveAssetPtr=&/Fort_Build_RepairStart_mix}}
 {id=WavePlayer_4, properties {soundWaveAssetPtr=&/Fort_Build_RepairStart_mix_opt}}
]
properties {
 bOverrideAttenuation=true
 bOverrideConcurrency=true
 maxDistance=3000
 soundClassObject=/Game/Sounds/FortSoundClassesAndModes/Fort_Player_Construction
 volumeMultiplier=0.7
 attenuationOverrides {falloffDistance=2600}
 concurrencyOverrides {resolutionRule=PreventNew}
}
