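# Fort_Build_Rotate_Cue is a UI sound cue, named for rotating a piece during building.
# It plays the front-end Fort_Menu_Misc_01 wave through a Modulator that raises and slightly randomizes pitch, reusing a menu sound for gameplay.
# It plays at low volume in the critical in-game UI sound class with spatialization disabled in its attenuation overrides.
# Its concurrency override allows a single instance, stopping the oldest when retriggered.
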SoundCue /Game/Sounds/Fort_UI_Sounds/Gameplay/Fort_Build_Rotate_Cue
root=Modulator_0
nodes [
 {id=Modulator_0, children [WavePlayer_1], properties {pitchMax=1.3, pitchMin=1.2}}
 {id=WavePlayer_1, properties {soundWaveAssetPtr=/Game/Sounds/Fort_UI_Sounds/FrontEnd/Fort_Menu_Misc_01}}
]
properties {
 bOverrideConcurrency=true
 maxDistance=2097152
 soundClassObject=&Fort_UI_InGame_Critical
 volumeMultiplier=0.18
 attenuationOverrides {bSpatialize=false, falloffDistance=2000}
 concurrencyOverrides {maxCount=1, resolutionRule=StopOldest}
}
