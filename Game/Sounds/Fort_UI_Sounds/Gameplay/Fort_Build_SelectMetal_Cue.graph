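# UI SoundCue for selecting metal as a build material, reusing a short metal demolition impact wave.
# A single wave player feeds a modulator as the root, at reduced volume in the critical in-game UI sound class.
# Playback is non-spatialized, and a concurrency override blocks new instances rather than cutting off playing ones.
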SoundCue /Game/Sounds/Fort_UI_Sounds/Gameplay/Fort_Build_SelectMetal_Cue
root=Modulator_0
nodes [
 {id=Modulator_0, children [WavePlayer_9]}
 {id=WavePlayer_9, properties {soundWaveAssetPtr=/Game/Sounds/Fort_Impact_Sounds/Demolition/Metal/Metal_Impact_Short_01}}
]
properties {
 bOverrideConcurrency=true
 maxDistance=2097152
 soundClassObject=&Fort_UI_InGame_Critical
 volumeMultiplier=0.24
 attenuationOverrides {bSpatialize=false, falloffDistance=2000, nonSpatializedRadiusStart=1000}
 concurrencyOverrides {resolutionRule=PreventNew}
}
