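# Sound cue for selecting rock as a build material in Fortnite, played as a critical in-game UI sound.
# A Mixer layers the rock start wave with the rock end wave, which follows after a short fixed Delay, and a Modulator at the root processes the mix.
# The cue raises pitch, lowers volume, disables spatialization in its attenuation overrides, and overrides concurrency to block new instances once the limit is reached.
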SoundCue /Game/Sounds/Fort_UI_Sounds/Gameplay/Fort_Build_SelectRock_Cue
root=Modulator_0
nodes [
 {id=Delay_0, children [WavePlayer_7], properties {delayMax=0.1, delayMin=0.1}}
 {id=Mixer_1, children [WavePlayer_8,Delay_0], properties {inputVolume [1,1]}}
 {id=Modulator_0, children [Mixer_1]}
 {id=WavePlayer_7, properties {soundWaveAssetPtr=&/Fort_Build_Rock_End}}
 {id=WavePlayer_8, properties {soundWaveAssetPtr=&/Fort_Build_Rock_Start}}
]
properties {
 bOverrideConcurrency=true
 maxDistance=2097152
 pitchMultiplier=1.2
 soundClassObject=&Fort_UI_InGame_Critical
 volumeMultiplier=0.45
 attenuationOverrides {bSpatialize=false, falloffDistance=2000, nonSpatializedRadiusStart=1000}
 concurrencyOverrides {resolutionRule=PreventNew}
}
