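# SoundCue for the sound that starts a build upgrade, playing on the player construction sound class at reduced volume.
# A single Fort_Build_Trap_01 wave player feeds a Modulator with a raised minimum pitch, with no random selection or layering.
# Overridden attenuation adds listener focus that quiets off-focus sources, plus stereo spread and distance falloff, while overridden concurrency prevents new instances once the limit is reached.
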SoundCue /Game/Sounds/Fort_UI_Sounds/Gameplay/Fort_Build_UpgradeStart_Cue
root=Modulator_0
nodes [
 {id=Modulator_0, children [WavePlayer_2], properties {pitchMin=1}}
 {id=WavePlayer_2, properties {soundWaveAssetPtr=&/Fort_Build_Trap_01}}
]
properties {
 bOverrideAttenuation=true
 bOverrideConcurrency=true
 maxDistance=2800
 soundClassObject=/Game/Sounds/FortSoundClassesAndModes/Fort_Player_Construction
 volumeMultiplier=0.5
 attenuationOverrides {bEnableListenerFocus=true, falloffDistance=2000, nonFocusVolumeAttenuation=0.3, stereoSpread=1000, attenuationShapeExtents {x=800}}
 concurrencyOverrides {resolutionRule=PreventNew}
}
