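# Sound cue for the in-game UI neutral message notification, playing the neutral update wave.
# A single wave player feeds the output at reduced volume, assigned to the non-critical in-game UI sound class.
# It overrides attenuation with its own falloff distance and caps concurrent instances, preventing new plays once the limit is reached.
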SoundCue /Game/Sounds/Fort_UI_Sounds/Gameplay/Fort_NeutralMessage_Cue
root=WavePlayer_27
nodes [{id=WavePlayer_27, properties {soundWaveAssetPtr=&/fort_UI_Neutral_Update}}]
properties {
 bOverrideAttenuation=true
 bOverrideConcurrency=true
 maxDistance=2000
 soundClassObject=&Fort_UI_InGame_NonCritical
 volumeMultiplier=0.2
 attenuationOverrides {falloffDistance=1600}
 concurrencyOverrides {maxCount=3, resolutionRule=PreventNew}
}
