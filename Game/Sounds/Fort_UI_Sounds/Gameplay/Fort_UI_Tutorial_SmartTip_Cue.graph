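# Sound cue for the tutorial smart tip UI sound in Fortnite gameplay.
# A single wave player feeds the UI_SmartTip_01 wave straight to the output at reduced volume, routed through the non-critical in-game UI sound class.
# Its concurrency override refuses new instances rather than cutting off ones already playing.
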SoundCue /Game/Sounds/Fort_UI_Sounds/Gameplay/Fort_UI_Tutorial_SmartTip_Cue
root=WavePlayer_0
nodes [{id=WavePlayer_0, properties {soundWaveAssetPtr=&/UI_SmartTip_01}}]
properties {
 maxDistance=2097152
 soundClassObject=&Fort_UI_InGame_NonCritical
 volumeMultiplier=0.35
 concurrencyOverrides {resolutionRule=PreventNew}
}
