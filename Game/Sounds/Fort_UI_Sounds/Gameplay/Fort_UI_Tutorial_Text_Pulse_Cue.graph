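# Sound cue for a UI tutorial text pulse, routed to the Fort_UI sound class at slightly reduced volume.
# A single wave player plays the UI_LeftScreen_Text_Pulse_01 wave, with no randomization or modulation.
# Its concurrency override allows one instance at a time and blocks new plays while one is active.
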SoundCue /Game/Sounds/Fort_UI_Sounds/Gameplay/Fort_UI_Tutorial_Text_Pulse_Cue
root=WavePlayer_1
nodes [{id=WavePlayer_1, properties {soundWaveAssetPtr=&/UI_LeftScreen_Text_Pulse_01}}]
properties {
 maxDistance=2097152
 soundClassObject=&Fort_UI
 volumeMultiplier=0.9
 concurrencyOverrides {maxCount=1, resolutionRule=PreventNew}
}
