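# One-shot stereo MetaSound that plays the building repair-start sound on the player construction sound class.
# On play, a random float sets a slight pitch shift and triggers a single wave player.
# The player's On Play then rolls a second random float, which scales left and right equally for small gain variation.
# It finishes with the wave, uses wall-build attenuation and overrides concurrency to cap simultaneous instances.
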MetaSoundSource /Game/Sounds/Fort_UI_Sounds/Gameplay/MSS_Building_RepairStart
inputs [UE.Source.OnPlay {type=Trigger}]
interfaces [UE.OutputFormat.Stereo:1.0, UE.Source.OneShot:1.0, UE.Source:1.1]
outputs [
 "UE.OutputFormat.Stereo.Audio:0" {type=Audio}
 "UE.OutputFormat.Stereo.Audio:1" {type=Audio}
 UE.Source.OneShot.OnFinished {type=Trigger}
]
pages [Default {
  input_defaults {UE.Source.OnPlay=null}
  nodes [
   n1 = UE.RandomFloat(Next=^UE.Source.OnPlay) {also {defaults {Max=0.05, Min=-0.05}}}
   WavePlayer_0 = UE.Wave Player:Stereo(Play=@n1.On Next, Pitch Shift=@n1.Value) {also {defaults {Wave Asset=&/Fort_Build_RepairStart_mix}, version [1,1]}}
   n2 = UE.RandomFloat(Next=@WavePlayer_0.On Play) {also {defaults {Max=1.05, Min=0.95}}}
   Multiply_0 = UE.Multiply:Audio by Float(AdditionalOperands=@n2.Value, PrimaryOperand=@WavePlayer_0.Out Right)
   Multiply_1 = UE.Multiply:Audio by Float(AdditionalOperands=@n2.Value, PrimaryOperand=@WavePlayer_0.Out Left)
  ]
  links [
   {dst="@UE.OutputFormat.Stereo.Audio:1", src=Multiply_0.Out}
   {dst="@UE.OutputFormat.Stereo.Audio:0", src=Multiply_1.Out}
   {dst="@UE.Source.OneShot.OnFinished", src=WavePlayer_0.On Finished}
  ]
 }]
properties {
 assetImportData=&self:AssetImportData
 attenuationSettings=/Game/Sounds/Attenuation/Fort_Build_Wall_Att
 audioPropertiesSheet=/Game/Sounds/Fort_UI_Sounds/Gameplay/BuildingSounds/Settings/APS_BuildingLoop_Parent
 bOverrideConcurrency=true
 maxDistance=2000
 numChannels=2
 outputFormat=Stereo
 sampleRate=48000
 soundClassObject=/Game/Sounds/FortSoundClassesAndModes/Fort_Player_Construction
 volume=0.7
 concurrencyOverrides {maxCount=5}
 concurrencySet [/Game/Sounds/Concurrency_Rules/Player_Wall_Build_Conc]
}
