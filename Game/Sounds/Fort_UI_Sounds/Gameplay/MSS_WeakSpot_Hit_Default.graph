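# One-shot stereo MetaSound Source for the default weak spot hit, playing a crit impact sound in the melee weapon sound class.
# On play, the Level input modulo the Sounds array length picks one of eight waves, so selection is deterministic and wraps around.
# The chosen wave feeds a stereo Wave Player, whose completion fires OnFinished.
# Attenuation and a concurrency rule limit its range and overlapping instances.
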MetaSoundSource /Game/Sounds/Fort_UI_Sounds/Gameplay/MSS_WeakSpot_Hit_Default
inputs [
 Level {type=Int32}
 Sounds {type=WaveAsset:Array}
 UE.Source.OnPlay {type=Trigger}
]
interfaces [UE.OutputFormat.Stereo:1.0, UE.Source.OneShot:1.0, UE.Source:1.1]
outputs [
 "UE.OutputFormat.Stereo.Audio:0" {type=Audio}
 "UE.OutputFormat.Stereo.Audio:1" {type=Audio}
 UE.Source.OneShot.OnFinished {type=Trigger}
]
pages [Default {
  input_defaults {
   Level=0
   UE.Source.OnPlay=null
   Sounds [
    &/crit_impact_secondpassNew_01
    &/crit_impact_secondpassNew_02
    &/crit_impact_secondpassNew_03
    &/crit_impact_secondpassNew_04
    &/crit_impact_secondpassNew_05
    &/crit_impact_secondpassNew_06
    &/crit_impact_secondpassNew_07
    &/crit_impact_secondpassNew_08
   ]
  }
  nodes [
   n1 = Array.Num:WaveAsset:Array(Array=^Sounds)
   WavePlayer_0 = UE.Wave Player:Stereo(
    Play=^UE.Source.OnPlay
    Wave Asset=Array.Get:WaveAsset:Array(Array=^Sounds, Trigger=^UE.Source.OnPlay, Index=UE.Modulo:Int32(PrimaryOperand=^Level, AdditionalOperands=@n1.Num).Out).Element
   ) {also {version [1,1]}}
  ]
  links [
   {dst="@UE.Source.OneShot.OnFinished", src=WavePlayer_0.On Finished}
   {dst="@UE.OutputFormat.Stereo.Audio:0", src=WavePlayer_0.Out Left}
   {dst="@UE.OutputFormat.Stereo.Audio:1", src=WavePlayer_0.Out Right}
  ]
 }]
properties {
 assetImportData=&self:AssetImportData
 attenuationSettings=/Game/Sounds/Attenuation/Fort_SoundCrack_Att
 loadingBehavior=PrimeOnLoad
 maxDistance=1300
 numChannels=2
 outputFormat=Stereo
 priority=10
 sampleRate=48000
 soundClassObject=&Fort_WeaponMelee_PSM
 volume=0.839
 concurrencySet [/Game/Sounds/Concurrency_Rules/FourMaximum_StopFurthestOldest_VolumeScale]
}
