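# Sound cue for the short call-out bubble UI sound, classed as non-critical in-game UI audio.
# A single wave player plays the bubble open wave through a Mixer root, whose second input is empty.
# The cue uses the in-game ping attenuation asset, so it follows that asset's spatial and falloff settings.
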SoundCue /Game/Sounds/Fort_UI_Sounds/Gameplay/SC_UI_CallOut_Bubble_Short
root=Mixer_0
nodes [
 {id=Mixer_0, children [WavePlayer_0,null], properties {inputVolume [1,1]}}
 {id=WavePlayer_0, properties {soundWaveAssetPtr=&/UI_CallOut_Bubble_Open}}
]
properties {
 attenuationSettings=/Game/Sounds/Attenuation/Fort_UI_Ping_Ingame_Att
 maxDistance=2097152
 soundClassObject=&Fort_UI_InGame_NonCritical
}
