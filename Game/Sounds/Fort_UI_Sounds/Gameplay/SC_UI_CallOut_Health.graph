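# Sound cue for the health call-out UI sound, spatialized with the in-game ping attenuation and routed to the non-critical in-game UI sound class.
# A Mixer layers the call-out bubble-open sound with a quieter, briefly delayed bandage pickup accent.
# The accent randomly picks one of two pickup waves, with slight upward pitch modulation for variation.
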SoundCue /Game/Sounds/Fort_UI_Sounds/Gameplay/SC_UI_CallOut_Health
root=Mixer_0
nodes [
 {id=Delay_1, children [Modulator_0], properties {delayMax=0.2, delayMin=0.2}}
 {id=Mixer_0, children [WavePlayer_0,Delay_1,null], properties {inputVolume [1,0.6,1]}}
 {id=Modulator_0, children [Random_0], properties {pitchMax=1.1, pitchMin=1}}
 {id=Random_0, children [WavePlayer_5,WavePlayer_6], properties {bShouldExcludeFromBranchCulling=true, weights [1,1]}}
 {id=WavePlayer_0, properties {soundWaveAssetPtr=&/UI_CallOut_Bubble_Open}}
 {id=WavePlayer_5, properties {soundWaveAssetPtr=/Game/Athena/Sounds/Items/Bandages/Bandage_Pickup_01}}
 {id=WavePlayer_6, properties {soundWaveAssetPtr=/Game/Athena/Sounds/Items/Bandages/Bandage_Pickup_02}}
]
properties {
 attenuationSettings=/Game/Sounds/Attenuation/Fort_UI_Ping_Ingame_Att
 maxDistance=2097152
 soundClassObject=&Fort_UI_InGame_NonCritical
 volumeMultiplier=1.2
}
