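# Sound cue for the in-game UI shields callout, played through the non-critical UI sound class with in-game ping attenuation.
# A mixer layers a callout bubble-open sound with a boosted second branch that starts after a short fixed delay.
# That delayed branch randomly picks one of three equally weighted small shield jar hand-grab sounds, so the callout reuses item audio to identify shields.
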SoundCue /Game/Sounds/Fort_UI_Sounds/Gameplay/SC_UI_CallOut_Shields
root=Mixer_0
nodes [
 {id=Delay_1, children [Random_1], properties {delayMax=0.2, delayMin=0.2}}
 {id=Mixer_0, children [WavePlayer_0,Delay_1,null], properties {inputVolume [1,3,1]}}
 {id=Random_1, children [WavePlayer_6,WavePlayer_7,WavePlayer_8], properties {bShouldExcludeFromBranchCulling=true, weights [1,1,1]}}
 {id=WavePlayer_0, properties {soundWaveAssetPtr=&/UI_CallOut_Bubble_Open}}
 {id=WavePlayer_6, properties {soundWaveAssetPtr=/Game/Athena/Sounds/Items/Shield_Jar_Small_Hand_Grab_03}}
 {id=WavePlayer_7, properties {soundWaveAssetPtr=/Game/Athena/Sounds/Items/Shield_Jar_Small_Hand_Grab_02}}
 {id=WavePlayer_8, properties {soundWaveAssetPtr=/Game/Athena/Sounds/Items/Shield_Jar_Small_Hand_Grab_01}}
]
properties {
 attenuationSettings=/Game/Sounds/Attenuation/Fort_UI_Ping_Ingame_Att
 maxDistance=2097152
 soundClassObject=&Fort_UI_InGame_NonCritical
 volumeMultiplier=1.2
}
