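# Sound cue that plays a level 1 crack impact sound for gameplay, classed as a melee weapon sound.
# A single wave player feeds the crit impact wave straight to the output, with no randomization or modulation.
# A dedicated crack attenuation asset handles distance falloff, and the cue primes on load with a slightly reduced volume.
# A shared concurrency set that stops the furthest, oldest instance with volume scaling manages overlapping playback.
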SoundCue /Game/Sounds/Fort_UI_Sounds/Gameplay/Sound_Crack_Level_01_Cue
root=WavePlayer_0
nodes [{id=WavePlayer_0, properties {soundWaveAssetPtr=&/crit_impact_secondpassNew_01}}]
properties {
 attenuationSettings=/Game/Sounds/Attenuation/Fort_SoundCrack_Att
 bPrimeOnLoad=true
 maxDistance=1300
 priority=10
 soundClassObject=&Fort_WeaponMelee_PSM
 volumeMultiplier=0.839
 concurrencyOverrides {maxCount=2}
 concurrencySet [/Game/Sounds/Concurrency_Rules/FourMaximum_StopFurthestOldest_VolumeScale]
}
