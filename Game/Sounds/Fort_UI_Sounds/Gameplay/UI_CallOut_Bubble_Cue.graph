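# Sound cue for the in-game UI call-out bubble, covering both its open and close sounds in a single trigger.
# A Mixer plays the open wave immediately alongside the close wave, which a fixed 1.6-second Delay holds back, so the close timing is baked into the cue.
# It uses the in-game ping attenuation and the non-critical in-game UI sound class.
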SoundCue /Game/Sounds/Fort_UI_Sounds/Gameplay/UI_CallOut_Bubble_Cue
root=Mixer_0
nodes [
 {id=Delay_0, children [WavePlayer_1], properties {delayMax=1.6, delayMin=1.6}}
 {id=Mixer_0, children [WavePlayer_0,Delay_0], properties {inputVolume [1,1]}}
 {id=WavePlayer_0, properties {soundWaveAssetPtr=&/UI_CallOut_Bubble_Open}}
 {id=WavePlayer_1, properties {soundWaveAssetPtr=&/UI_CallOut_Bubble_Close}}
]
properties {
 attenuationSettings=/Game/Sounds/Attenuation/Fort_UI_Ping_Ingame_Att
 maxDistance=2097152
 soundClassObject=&Fort_UI_InGame_NonCritical
}
