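# Sound cue for the UI feedback that plays when the player cannot build, as its name indicates.
# A single wave player feeds the UI_CantBuild01 wave straight to the root, at reduced volume, in the critical in-game UI sound class.
# A concurrency override allows one instance at a time and blocks new plays while one is active, so repeated triggers do not stack.
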SoundCue /Game/Sounds/Fort_UI_Sounds/Gameplay/UI_CantBuild_Cue
root=WavePlayer_2
nodes [{id=WavePlayer_2, properties {soundWaveAssetPtr=&/UI_CantBuild01}}]
properties {
 bOverrideConcurrency=true
 maxDistance=2097152
 soundClassObject=&Fort_UI_InGame_Critical
 volumeMultiplier=0.45
 attenuationOverrides {falloffDistance=1600}
 concurrencyOverrides {maxCount=1, resolutionRule=PreventNew}
}
