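# UI_DropItem_Cue is a sound cue for the in-game UI sound heard when an item is dropped.
# A single wave player feeds the UI_DropItem02 wave straight to the output, with no randomization or modulation.
# It is classed as non-critical in-game UI audio and overrides attenuation to play non-spatialized without distance falloff.
# Its concurrency override prevents new instances from starting once the limit is reached.
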SoundCue /Game/Sounds/Fort_UI_Sounds/Gameplay/UI_DropItem_Cue
root=WavePlayer_9
nodes [{id=WavePlayer_9, properties {soundWaveAssetPtr=&/UI_DropItem02}}]
properties {
 bOverrideAttenuation=true
 bOverrideConcurrency=true
 maxDistance=2097152
 soundClassObject=&Fort_UI_InGame_NonCritical
 attenuationOverrides {bAttenuate=false, bSpatialize=false, falloffDistance=1100, nonSpatializedRadiusStart=200}
 concurrencyOverrides {resolutionRule=PreventNew}
}
