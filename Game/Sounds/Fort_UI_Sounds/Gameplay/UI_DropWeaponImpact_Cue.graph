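# Sound cue for the impact of a dropped weapon, assigned to the loot-collect sound class.
# A mixer layers one random pick from three generic drop-impact waves with a quieter, briefly and randomly delayed pick from three others, forming a varied double hit.
# Its own attenuation override limits audible range, and a concurrency override caps simultaneous instances by blocking new ones.
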SoundCue /Game/Sounds/Fort_UI_Sounds/Gameplay/UI_DropWeaponImpact_Cue
root=Mixer_0
nodes [
 {id=Delay_0, children [Random_1], properties {delayMax=0.25, delayMin=0.1}}
 {id=Mixer_0, children [Random_0,Delay_0], properties {inputVolume [1,0.4]}}
 {id=Random_0, children [WavePlayer_0,WavePlayer_6,WavePlayer_2], properties {weights [1,1,1]}}
 {id=Random_1, children [WavePlayer_1,WavePlayer_7,WavePlayer_8], properties {weights [1,1,1]}}
 {id=WavePlayer_0, properties {soundWaveAssetPtr=&/Weapon_Generic_Drop_Impact_01}}
 {id=WavePlayer_1, properties {soundWaveAssetPtr=&/Weapon_Generic_Drop_Impact_02}}
 {id=WavePlayer_2, properties {soundWaveAssetPtr=&/Weapon_Generic_Drop_Impact_03}}
 {id=WavePlayer_6, properties {soundWaveAssetPtr=&/Weapon_Generic_Drop_Impact_04}}
 {id=WavePlayer_7, properties {soundWaveAssetPtr=&/Weapon_Generic_Drop_Impact_05}}
 {id=WavePlayer_8, properties {soundWaveAssetPtr=&/Weapon_Generic_Drop_Impact_06}}
]
properties {
 bOverrideAttenuation=true
 bOverrideConcurrency=true
 maxDistance=1000
 soundClassObject=&Fort_Loot_Collect
 volumeMultiplier=1
 attenuationOverrides {falloffDistance=1000, attenuationShapeExtents {x=0}}
 concurrencyOverrides {maxCount=3, resolutionRule=PreventNew}
}
