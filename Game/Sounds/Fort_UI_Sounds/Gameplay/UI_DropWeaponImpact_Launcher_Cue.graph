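# Sound cue for the impact of a dropped launcher weapon, assigned to the loot collect sound class.
# A single launcher impact wave plays through a Modulator node left at its default settings.
# The cue overrides attenuation with a short falloff distance and overrides concurrency to cap simultaneous instances, blocking new plays once the limit is reached.
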SoundCue /Game/Sounds/Fort_UI_Sounds/Gameplay/UI_DropWeaponImpact_Launcher_Cue
root=Modulator_0
nodes [
 {id=Modulator_0, children [WavePlayer_0]}
 {id=WavePlayer_0, properties {soundWaveAssetPtr=&/Weapon_Generic_Drop_Impact_Launcher_Impt_01}}
]
properties {
 bOverrideAttenuation=true
 bOverrideConcurrency=true
 maxDistance=1000
 soundClassObject=&Fort_Loot_Collect
 volumeMultiplier=1
 attenuationOverrides {falloffDistance=1000, attenuationShapeExtents {x=0}}
 concurrencyOverrides {maxCount=3, resolutionRule=PreventNew}
}
