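# Sound cue for the impact sound of a dropped pistol, routed to the Fort_Loot_Collect sound class.
# A single wave player feeds a Modulator left at default settings, then a one-input Mixer at the root.
# The cue overrides attenuation with a short falloff and boosts volume slightly.
# Its concurrency override caps simultaneous instances at a small count and prevents new ones beyond that.
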SoundCue /Game/Sounds/Fort_UI_Sounds/Gameplay/UI_DropWeaponImpact_Pistol_Cue
root=Mixer_0
nodes [
 {id=Mixer_0, children [Modulator_1], properties {inputVolume [1]}}
 {id=Modulator_1, children [WavePlayer_1]}
 {id=WavePlayer_1, properties {soundWaveAssetPtr=&/Weapon_Generic_Drop_Impact_Pistol}}
]
properties {
 bOverrideAttenuation=true
 bOverrideConcurrency=true
 maxDistance=1000
 soundClassObject=&Fort_Loot_Collect
 volumeMultiplier=1.25
 attenuationOverrides {falloffDistance=1000, attenuationShapeExtents {x=0}}
 concurrencyOverrides {maxCount=3, resolutionRule=PreventNew}
}
