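# Sound cue for the impact of a dropped rifle weapon, using generic heavy rifle drop sounds.
# A Random node picks one of three equally weighted wave variations.
# It plays under the loot-collect sound class, with overridden attenuation for short-range falloff from the source.
# Concurrency is overridden to cap simultaneous instances and block new plays once the limit is reached.
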SoundCue /Game/Sounds/Fort_UI_Sounds/Gameplay/UI_DropWeaponImpact_Rifle_Cue
root=Random_1
nodes [
 {id=Random_1, children [WavePlayer_6,WavePlayer_7,WavePlayer_8], properties {weights [1,1,1]}}
 {id=WavePlayer_6, properties {soundWaveAssetPtr=&/Weapon_Generic_Drop_Impact_Rifle_Hv_01}}
 {id=WavePlayer_7, properties {soundWaveAssetPtr=&/Weapon_Generic_Drop_Impact_Rifle_Hv_02}}
 {id=WavePlayer_8, properties {soundWaveAssetPtr=&/Weapon_Generic_Drop_Impact_Rifle_Hv_03}}
]
properties {
 bOverrideAttenuation=true
 bOverrideConcurrency=true
 maxDistance=1000
 soundClassObject=&Fort_Loot_Collect
 volumeMultiplier=1
 attenuationOverrides {falloffDistance=1000, attenuationShapeExtents {x=0}}
 concurrencyOverrides {maxCount=3, resolutionRule=PreventNew}
}
