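# UI sound cue for a "can't do" error ping during gameplay, routed through the Fort_UI sound class.
# A mixer layers a boosted pistol foley sample, raised to a fixed higher pitch by a modulator, with a quieter default error ping wave.
# Both layers play together with no branching, and a concurrency override caps overlapping instances at two by stopping the oldest.
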SoundCue /Game/Sounds/Fort_UI_Sounds/Gameplay/UI_Error_Ping_Cant_Do_NoBranch_Cue
root=Mixer_0
nodes [
 {id=Mixer_0, children [Modulator_0,WavePlayer_0], properties {inputVolume [1.25,0.5]}}
 {id=Modulator_0, children [WavePlayer_2], properties {pitchMax=1.3, pitchMin=1.3}}
 {id=WavePlayer_0, properties {soundWaveAssetPtr=&/UI_Error_Ping_Cant_Do_Default_01}}
 {id=WavePlayer_2, properties {soundWaveAssetPtr=/Game/Sounds/Fort_Weapons/pistol/pistol_foley/auto_sdk_end_01}}
]
properties {
 bOverrideConcurrency=true
 maxDistance=2097152
 soundClassObject=&Fort_UI
 volumeMultiplier=1
 concurrencyOverrides {maxCount=2, resolutionRule=StopOldest}
}
