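# Sound cue for the gameplay UI sound that, by its name, signals not having enough resources.
# A single wave player feeds UI_NotEnoughResources03 straight to the output at reduced volume, in the critical in-game UI sound class.
# Its own concurrency override refuses new instances once the limit is reached, and its attenuation override turns off spatialization.
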SoundCue /Game/Sounds/Fort_UI_Sounds/Gameplay/UI_NotEnoughResources_Cue
root=WavePlayer_9
nodes [{id=WavePlayer_9, properties {soundWaveAssetPtr=&/UI_NotEnoughResources03}}]
properties {
 bOverrideConcurrency=true
 maxDistance=2097152
 soundClassObject=&Fort_UI_InGame_Critical
 volumeMultiplier=0.45
 attenuationOverrides {bSpatialize=false, falloffDistance=2000, nonSpatializedRadiusStart=200}
 concurrencyOverrides {resolutionRule=PreventNew}
}
