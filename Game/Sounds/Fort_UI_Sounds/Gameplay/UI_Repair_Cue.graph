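# Sound cue for the repair UI sound during gameplay, assigned to the critical in-game UI sound class.
# A single wave player feeds UI_Repair01 straight to the output, with no randomization or mixing.
# The cue overrides attenuation with its own falloff distance, and overrides concurrency so new plays are blocked at the limit instead of cutting off ones already playing.
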SoundCue /Game/Sounds/Fort_UI_Sounds/Gameplay/UI_Repair_Cue
root=WavePlayer_1
nodes [{id=WavePlayer_1, properties {soundWaveAssetPtr=&/UI_Repair01}}]
properties {
 bOverrideAttenuation=true
 bOverrideConcurrency=true
 maxDistance=2000
 soundClassObject=&Fort_UI_InGame_Critical
 attenuationOverrides {falloffDistance=1600}
 concurrencyOverrides {resolutionRule=PreventNew}
}
