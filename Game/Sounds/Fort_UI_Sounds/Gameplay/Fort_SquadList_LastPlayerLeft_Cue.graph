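# Sound cue for the squad list "last player left" event, judging by its name, built from front-end UI click sounds.
# A mixer layers a scroll click and a main button click at equal, reduced levels, with a low overall cue volume.
# It uses the important in-game mission sound class and overrides attenuation.
# Its concurrency override allows one instance at a time and blocks new plays while it sounds.
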SoundCue /Game/Sounds/Fort_UI_Sounds/Gameplay/Fort_SquadList_LastPlayerLeft_Cue
root=Mixer_0
nodes [
 {id=Mixer_0, children [WavePlayer_2,WavePlayer_0], properties {inputVolume [0.75,0.75]}}
 {id=WavePlayer_0, properties {soundWaveAssetPtr=/Game/Sounds/Fort_UI_Sounds/FrontEnd/New_FrontEnd/UI_FN_Button_Click_Main_01}}
 {id=WavePlayer_2, properties {soundWaveAssetPtr=/Game/Sounds/Fort_UI_Sounds/FrontEnd/New_FrontEnd/UI_FN_ScrollClick02}}
]
properties {
 bOverrideAttenuation=true
 bOverrideConcurrency=true
 maxDistance=2000
 soundClassObject=&Fort_Mission_Ingame_Important_PSM
 volumeMultiplier=0.3
 attenuationOverrides {falloffDistance=1600}
 concurrencyOverrides {maxCount=1, resolutionRule=PreventNew}
}
