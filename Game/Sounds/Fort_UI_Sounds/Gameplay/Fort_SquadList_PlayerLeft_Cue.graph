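# UI sound cue that, by its name, plays when a player leaves the squad list.
# A mixer layers a slightly quieter scroll click with a main button click that a modulator pitches down by a fixed amount, both reused front-end sounds.
# The cue plays at reduced volume through an important in-game mission sound class, and its concurrency override allows one instance at a time, rejecting new ones while it plays.
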SoundCue /Game/Sounds/Fort_UI_Sounds/Gameplay/Fort_SquadList_PlayerLeft_Cue
root=Mixer_1
nodes [
 {id=Mixer_1, children [WavePlayer_2,Modulator_0], properties {inputVolume [0.75,1]}}
 {id=Modulator_0, children [WavePlayer_0], properties {pitchMax=0.75, pitchMin=0.75}}
 {id=WavePlayer_0, properties {soundWaveAssetPtr=/Game/Sounds/Fort_UI_Sounds/FrontEnd/New_FrontEnd/UI_FN_Button_Click_Main_01}}
 {id=WavePlayer_2, properties {soundWaveAssetPtr=/Game/Sounds/Fort_UI_Sounds/FrontEnd/New_FrontEnd/UI_FN_ScrollClick02}}
]
properties {
 bOverrideAttenuation=true
 bOverrideConcurrency=true
 maxDistance=2000
 soundClassObject=&Fort_Mission_Ingame_Important_PSM
 volumeMultiplier=0.3
 attenuationOverrides {falloffDistance=1600}
 concurrencyOverrides {maxCount=1, resolutionRule=PreventNew}
}
