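# UI_QuickMenu_Close_Cue is a UI sound cue, named for closing the quick menu during gameplay.
# It plays the player survey pop-up wave through a Modulator that randomly raises its pitch slightly on each play.
# The cue routes to the Fort_UI sound class at reduced volume, with a concurrency override rule that prevents new instances rather than stopping existing ones.
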SoundCue /Game/Sounds/Fort_UI_Sounds/Gameplay/UI_QuickMenu_Close_Cue
root=Modulator_0
nodes [
 {id=Modulator_0, children [WavePlayer_1], properties {pitchMax=1.15, pitchMin=1.1}}
 {id=WavePlayer_1, properties {soundWaveAssetPtr=/Game/Sounds/Fort_UI_Sounds/FrontEnd/Fort_UI_PlayerSurvey_PopUp_01}}
]
properties {
 maxDistance=2097152
 soundClassObject=&Fort_UI
 volumeMultiplier=0.6
 concurrencyOverrides {resolutionRule=PreventNew}
}
