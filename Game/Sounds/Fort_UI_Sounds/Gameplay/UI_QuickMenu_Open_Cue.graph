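# UI sound cue, named for the quick menu opening in gameplay, that layers two reused front-end UI sounds in the Fort_UI sound class.
# A Mixer plays the survey pop-up wave at full level alongside a much quieter, briefly delayed options-menu wave.
# The mix feeds a root Modulator at default settings for slight pitch and volume variation, with the cue's overall volume reduced.
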SoundCue /Game/Sounds/Fort_UI_Sounds/Gameplay/UI_QuickMenu_Open_Cue
root=Modulator_0
nodes [
 {id=Delay_0, children [WavePlayer_5], properties {delayMax=0.08, delayMin=0.08}}
 {id=Mixer_0, children [Delay_0,WavePlayer_1], properties {inputVolume [0.2,1]}}
 {id=Modulator_0, children [Mixer_0]}
 {id=WavePlayer_1, properties {soundWaveAssetPtr=/Game/Sounds/Fort_UI_Sounds/FrontEnd/Fort_UI_PlayerSurvey_PopUp_01}}
 {id=WavePlayer_5, properties {soundWaveAssetPtr=/Game/Sounds/Fort_UI_Sounds/FrontEnd/New_FrontEnd/Fortnite_UI_Options_Menu_03}}
]
properties {
 maxDistance=2097152
 soundClassObject=&Fort_UI
 volumeMultiplier=0.7
 concurrencyOverrides {resolutionRule=PreventNew}
}
